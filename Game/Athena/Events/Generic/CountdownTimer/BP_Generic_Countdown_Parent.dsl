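; Base Blueprint for event countdown displays with a screen, hologram and skybeam, shown or hidden as a timer manager turns active or inactive.
; BeginPlay polls the manager's soft reference every second until valid, then binds its events to child-overridable handlers.
; The construction script applies material and HLOD overrides, positioning, visibility, dynamic material parameters and debug MPC states.
; Helper functions start and stop countdown audio with a sound mix, though nothing here calls them.
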(event Custom|TryFindTimerManager
  (bind _timermanagersoftref (Variables|Default|GetTimerManagerSoftRef))
  (bind _output (Utilities|ResolveSoftReference _timermanagersoftref))
  ;[0 0 0] If the timer manager soft ref is set, wait for the object it's referencing to become valid, then initialize
  (Utilities|IsValid _output
    (:Is%20Valid
      (Variables|Default|SetCachedTimerManager _output)
      (CallFunction|InitializewithTimerManager))
    (:Is%20Not%20Valid
      (Utilities|Time|SetTimerbyEvent (EventDispatchers|CreateEvent :!event Custom|TryFindTimerManager) 1.0))))

(event EventBeginPlay
  (bind _timermanagersoftref (Variables|Default|GetTimerManagerSoftRef))
  (bind _output (Utilities|ResolveSoftReference _timermanagersoftref))
  ;[0 0 0] If the timer manager soft ref is set, wait for the object it's referencing to become valid, then initialize
  (if (Utilities|IsValidSoftObjectReference _timermanagersoftref)
    (Utilities|IsValid _output
      (:Is%20Valid
        (Variables|Default|SetCachedTimerManager _output)
        (CallFunction|InitializewithTimerManager))
      (:Is%20Not%20Valid
        (Utilities|Time|SetTimerbyEvent (EventDispatchers|CreateEvent :!event Custom|TryFindTimerManager) 1.0)))))

(event EventEndPlay
  (params (EndPlayReason :type EEndPlayReason%20Enum :default "Destroyed"))
  (bind _cachedtimermanager (Variables|Default|GetCachedTimerManager))
  ;[0 0 0] Clean up timer events
  (Utilities|IsValid _cachedtimermanager
    (:Is%20Valid
      (Default|UnbindEventfromOnTimerFinished _cachedtimermanager (EventDispatchers|CreateEvent :!event Custom|OnTimerFinished))
      (Default|UnbindEventfromOnTimerStateChange _cachedtimermanager (EventDispatchers|CreateEvent :!event Custom|OnTimerStateChanged))
      (Default|UnbindEventfromOnTimerActive _cachedtimermanager (EventDispatchers|CreateEvent :!event Custom|OnTimerActive))
      (Default|UnbindEventfromOnTimerInactive _cachedtimermanager (EventDispatchers|CreateEvent :!event Custom|OnTimerInactive)))))

(event Custom|OnTimerActive
  ;[1 1 1] To be overidden by child classes
  (Rendering|SetActorHiddenInGame false))

(event Custom|OnTimerStateChanged
  (params (NewTimerState :type Integer)))

(event Custom|OnTimerFinished)

(event Custom|OnTimerInactive
  ;[1 1 1] To be overidden by child classes
  (Rendering|SetActorHiddenInGame true))

(fn HandleHLODMaterialOverrides
  (bind _hlod_override_skybeam (Variables|MaterialOverrides|GetHLOD%20Override_SkyBeam))
  (bind _hlod_override_hologram (Variables|MaterialOverrides|GetHLOD%20Override_Hologram))
  (bind _hlodelementindex (Variables|MaterialOverrides|GetHLODElementIndex))
  (Utilities|FlowControl|Sequence
    (:then_0
      (Utilities|IsValid _hlod_override_hologram
        (:Is%20Valid
          (CallFunction|OverrideComponentMaterial (Variables|Default|GetHologramCountdown) _hlod_override_hologram _hlodelementindex))))
    (:then_1
      (Utilities|IsValid _hlod_override_skybeam
        (:Is%20Valid
          (CallFunction|OverrideComponentMaterial (Variables|Default|GetHologramSkybeam) _hlod_override_skybeam _hlodelementindex))))))

(fn HandleMaterialPropertyOverrides
  (bind _beam_override (Variables|MaterialOverrides|GetBeam_Override))
  (bind _countdownhologram_override (Variables|MaterialOverrides|GetCountdownHologram_Override))
  (bind _countdownscreen_override (Variables|MaterialOverrides|GetCountdownScreen_Override))
  (Utilities|FlowControl|Sequence
    (:then_0
      (Utilities|IsValid _countdownscreen_override
        (:Is%20Valid
          (CallFunction|OverrideComponentMaterial (Variables|Default|GetCountdownScreen) _countdownscreen_override))))
    (:then_1
      (Utilities|IsValid _beam_override
        (:Is%20Valid
          (CallFunction|OverrideComponentMaterial (Variables|Default|GetHologramSkybeam) _beam_override))))
    (:then_2
      (Utilities|IsValid _countdownhologram_override
        (:Is%20Valid
          (CallFunction|OverrideComponentMaterial (Variables|Default|GetHologramCountdown) _countdownhologram_override))))))

(fn HandleSetupMIDs
  (Utilities|FlowControl|Sequence
    (:then_0
      ;[0 0 0] set MIDs
      (bind _returnvalue (Rendering|Material|CreateDynamicMaterialInstance (Variables|Default|GetCountdownScreen) 0 0 "None" :!class /Script/Engine.PrimitiveComponent))
      (Variables|MaterialOverrides|SetMI_CountdownScreen _returnvalue)
      (bind _returnvalue_2 (Rendering|Material|CreateDynamicMaterialInstance (Variables|Default|GetHologramCountdown) 0 0 "None" :!class /Script/Engine.PrimitiveComponent))
      (Variables|MaterialOverrides|SetMI_Countdown_Hologram _returnvalue_2)
      (bind _returnvalue_3 (Rendering|Material|CreateDynamicMaterialInstance (Variables|Default|GetHologramSkybeam) 0 0 "None" :!class /Script/Engine.PrimitiveComponent))
      (Variables|MaterialOverrides|SetMI_Skybeam _returnvalue_3))
    (:then_1
      ;[0 0 0] number params
      (Rendering|Material|SetScalarParameterValue (Variables|MaterialOverrides|GetMI_CountdownScreen) "NumberofDigits" (Math|Conversions|ToFloat(Integer) (Variables|BaseSetup|GetNumberofDigits)) :!class /Script/Engine.MaterialInstanceDynamic)
      (Rendering|Material|SetScalarParameterValue (Variables|MaterialOverrides|GetMI_CountdownScreen) "u_NumberPaddingScale" (Variables|Clock|GetUV_Padding_x) :!class /Script/Engine.MaterialInstanceDynamic)
      (Rendering|Material|SetScalarParameterValue (Variables|MaterialOverrides|GetMI_CountdownScreen) "v_NumberPaddingScale" (Variables|Clock|GetUV_Padding_y) :!class /Script/Engine.MaterialInstanceDynamic)
      (Rendering|Material|SetScalarParameterValue (Variables|MaterialOverrides|GetMI_CountdownScreen) "u_NumberPaddingOffset" (Variables|Clock|GetUV_Offset_x) :!class /Script/Engine.MaterialInstanceDynamic)
      (Rendering|Material|SetScalarParameterValue (Variables|MaterialOverrides|GetMI_CountdownScreen) "v_NumberPaddingOffset" (Variables|Clock|GetUV_Offset_y) :!class /Script/Engine.MaterialInstanceDynamic))
    (:then_2
      (Rendering|Material|SetScalarParameterValue (Variables|MaterialOverrides|GetMI_CountdownScreen) "Vignette" (Variables|Hologram|GetHologramVignette) :!class /Script/Engine.MaterialInstanceDynamic)
      (Rendering|Material|SetScalarParameterValue (Variables|MaterialOverrides|GetMI_Countdown_Hologram) "DistanceFadeLength" (Variables|Hologram|GetDistanceFadeLength) :!class /Script/Engine.MaterialInstanceDynamic)
      (Rendering|Material|SetScalarParameterValue (Variables|MaterialOverrides|GetMI_Countdown_Hologram) "DistanceFadeOffset" (Variables|Hologram|GetDistanceFadeOffset) :!class /Script/Engine.MaterialInstanceDynamic))
    (:then_3
      (Rendering|Material|SetTextureParameterValue (Variables|MaterialOverrides|GetMI_CountdownScreen) "DHMS" "/Game/Athena/Events/Generic/CountdownTimer/Textures/RT_LocalizedCountdownText.RT_LocalizedCountdownText")
      (Rendering|Material|SetScalarParameterValue (Variables|MaterialOverrides|GetMI_Countdown_Hologram) "ScaleX" (Variables|Hologram|GetScaleHologram_x) :!class /Script/Engine.MaterialInstanceDynamic)
      (Rendering|Material|SetScalarParameterValue (Variables|MaterialOverrides|GetMI_Countdown_Hologram) "ScaleY" (Variables|Hologram|GetScaleHologram_y) :!class /Script/Engine.MaterialInstanceDynamic)
      (Rendering|Material|SetScalarParameterValue (Variables|MaterialOverrides|GetMI_Skybeam) "Width2" (Variables|Hologram|GetScaleHologram_x) :!class /Script/Engine.MaterialInstanceDynamic))))

(fn InitializeWithTimerManager
  (bind _cachedtimermanager (Variables|Default|GetCachedTimerManager))
  ;[0 0 0] Bind to timer manager events, check if the timer is already active before binding to the on timer activate event
  (Utilities|IsValid _cachedtimermanager
    (:Is%20Valid
      (Default|BindEventtoOnTimerStateChange _cachedtimermanager (EventDispatchers|CreateEvent :!event OnTimerStateChanged))
      (Default|BindEventtoOnTimerFinished _cachedtimermanager (EventDispatchers|CreateEvent :!event OnTimerFinished))
      (Default|BindEventtoOnTimerInactive _cachedtimermanager (EventDispatchers|CreateEvent :!event OnTimerInactive))
      (if (Variables|/Game/Athena/Events/Generic/CountdownTimer/BP_Generic_TimerManager.BP_Generic_TimerManager_C|GetbTimerActive _cachedtimermanager)
        (CallFunction|OnTimerActive)
        (if (Variables|/Script/FortniteGame.FortRuntimeOptions|GetbEnableLobbyCountdownBasedOnLinkData (Options|GetRuntimeOptions))
          _))
      (Default|BindEventtoOnTimerActive _cachedtimermanager (EventDispatchers|CreateEvent :!event OnTimerActive)))))

(fn OverrideComponentMaterial
  (params (InComponent :type Primitive%20Component%20Object%20Reference) (OverrideMaterial :type Material%20Interface%20Object%20Reference) (ElementIndex :type Integer))
  (if (and (Utilities|IsValid InComponent :!class /Script/Engine.KismetSystemLibrary) (Utilities|IsValid OverrideMaterial :!class /Script/Engine.KismetSystemLibrary))
    (Rendering|Material|SetMaterial InComponent ElementIndex OverrideMaterial)
    (return true)
    (else
      (return false))))

(fn SetHologramRelativeTransform
  (bind _hologramheight (Variables|Hologram|GetHologramHeight))
  (bind _returnvalue (Math|Vector|MakeVector 0.0 0.0 _hologramheight))
  (bind _hologramskybeam (Variables|Default|GetHologramSkybeam))
  (Transformation|SetRelativeLocation (Variables|Default|GetHologramCountdown) _returnvalue false false)
  (Transformation|SetRelativeLocation _hologramskybeam (/ _returnvalue 2.0) false false)
  (Transformation|SetRelativeScale3D _hologramskybeam (Math|Vector|MakeVector 1.0 1.0 (/ (.z _returnvalue) 1.2e+02))))

(fn SetupVisibility
  (bind _hologramvisible (Variables|BaseSetup|GetHologramVisible))
  (bind _showholograminhlod (Variables|BaseSetup|GetShowHologramInHLOD))
  (bind _returnvalue (and _showholograminhlod _hologramvisible))
  (bind _hologramcountdown (Variables|Default|GetHologramCountdown))
  (bind _countdownscreen (Variables|Default|GetCountdownScreen))
  (Utilities|FlowControl|Sequence
    (:then_0
      ;[1 1 1] SetVisiblity of elements
      (Rendering|SetVisibility (Variables|Default|GetCountdownScreen) (Variables|BaseSetup|GetScreenVisible) false)
      (Variables|/Script/Engine.PrimitiveComponent|SetbEnableAutoLODGeneration (and (Variables|BaseSetup|GetShowScreenInHLOD) (Variables|BaseSetup|GetScreenVisible)) _countdownscreen)
      (HLOD|SetExcludedFromHLODLevel _countdownscreen "HLOD0" (Variables|BaseSetup|GetExcludefromMediumHLOD)))
    (:then_1
      (Rendering|SetVisibility (Variables|Default|GetHologramCountdown) (Variables|BaseSetup|GetHologramVisible) true)
      (Variables|/Script/Engine.PrimitiveComponent|SetbEnableAutoLODGeneration _returnvalue _hologramcountdown)
      (Variables|/Script/Engine.PrimitiveComponent|SetbEnableAutoLODGeneration _returnvalue (Variables|Default|GetHologramSkybeam))
      (HLOD|SetExcludedFromHLODLevel _hologramcountdown "HLOD0" (Variables|BaseSetup|GetExcludefromMediumHLOD)))))

(fn StartCountdownAudio
  (Variables|Default|GetCountdownAudioComp
    (:then)
    (:else
      (bind _returnvalue (Audio|SpawnSoundatLocation (Variables|Audio|GetCountdownAsset) (Math|Vector|SelectVector (Variables|Audio|GetAudioLocationOverride) (Transformation|GetActorLocation) (Variables|Audio|GetbUseAudioLocationOverride))))
      (Variables|Default|SetCountdownAudioComp _returnvalue)
      (bind _countdownsoundmix (Variables|Audio|GetCountdownSoundMix)
        (:then
          (Audio|PushSoundMixModifier _countdownsoundmix))
        (:else)))))

(fn StopCountdownAudio
  (bind _countdownaudiocomp (Variables|Default|GetCountdownAudioComp)
    (:then
      (Audio|Components|Audio|FadeOut _countdownaudiocomp (Variables|Audio|GetCountdownAudioFadeOut) 0.0 "Linear" :!class /Script/Engine.AudioComponent))
    (:else
      _))
  (bind _countdownsoundmix (Variables|Audio|GetCountdownSoundMix)
    (:then
      (Audio|PopSoundMixModifier _countdownsoundmix))
    (:else
      _))
  (bind _countdownendasset (Variables|Audio|GetCountdownEndAsset)
    (:then
      (Audio|PlaySoundatLocation _countdownendasset (Transformation|GetWorldLocation _countdownaudiocomp)))
    (:else)))

(fn ConstructionScript
  (Utilities|FlowControl|Sequence
    (:then_0
      (CallFunction|HandleMaterialPropertyOverrides))
    (:then_1
      (CallFunction|HandleHLODMaterialOverrides))
    (:then_2
      (CallFunction|SetHologramRelativeTransform))
    (:then_3
      (CallFunction|SetupVisibility))
    (:then_4
      (CallFunction|HandleSetupMIDs))
    (:then_5
      ;[1 1 1] Boundscale
      (Rendering|SetBoundsScale (Variables|Default|GetHologramCountdown) (/ (Math|Float|Max(Float) (Variables|Hologram|GetScaleHologram_x) (Variables|Hologram|GetScaleHologram_y)) 1e+02)))
    (:then_6
      ;[1 1 1] Debug different States
      (if (Variables|Debug|GetDebug_ShowTimer)
        (Rendering|Material|SetScalarParameterValue "/Game/Athena/Events/Generic/CountdownTimer/Materials/MPC_Countdown.MPC_Countdown" "ShowCountdown" 1.0 :!class /Script/Engine.KismetMaterialLibrary)
        (Rendering|Material|SetScalarParameterValue "/Game/Athena/Events/Generic/CountdownTimer/Materials/MPC_Countdown.MPC_Countdown" "ShowCountdownPhase2" 1.0 :!class /Script/Engine.KismetMaterialLibrary)
        (if (Variables|Debug|GetDebug_ShowTimerEnded)
          (Rendering|Material|SetScalarParameterValue "/Game/Athena/Events/Generic/CountdownTimer/Materials/MPC_Countdown.MPC_Countdown" "ShowCountdownPhase3" 1.0 :!class /Script/Engine.KismetMaterialLibrary)
          (else
            (Rendering|Material|SetScalarParameterValue "/Game/Athena/Events/Generic/CountdownTimer/Materials/MPC_Countdown.MPC_Countdown" "ShowCountdownPhase3" :!class /Script/Engine.KismetMaterialLibrary)))))))
