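; Generic event countdown Blueprint that triggers optional pre-event effects as its timer nears zero.
; It listens to the cached timer manager's seconds-remaining updates until the scalable time buffer is reached, or unbinds if effects are disabled.
; Outside the SafeZones phase it polls on a looping timer, then broadcasts a named event through the global GameplayEventRouter.
; Timer finish calls RemovePreEventEffects, which is empty here.
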(event EventBeginPlay
  (bind _cachedtimermanager (Variables|Default|GetCachedTimerManager))
  ; Bind to Timer Manager while it ticks down, Parent disconnects from this event when entering Blinking Mode.
  (|CallParentFunction)
  (Default|BindEventtoOnSecondsRemainingUpdated _cachedtimermanager (EventDispatchers|CreateEvent :!event Custom|CheckForPreEventGate))
  (Default|BindEventtoOnTimerFinished _cachedtimermanager (EventDispatchers|CreateEvent :!event OnTimerFinished)))

(event Custom|CheckForPreEventGate
  (params (SecondsRemaining :type Integer))
  (bind _cachedtimermanager (Variables|Default|GetCachedTimerManager))
  ;[0 0.15 0.07284 0.5] While timer counts down, check if we are using Pre Event Effects. If not, unbind from Timer Manager
  (Variables|Default|SetSecondsRemaining (Math|Conversions|ToFloat(Integer) SecondsRemaining))
  (if (Variables|Default|GetUsePreEventEffects)
    (if (<= (Variables|Default|GetSecondsRemaining) (ScalableFloat|GetValueatLevelwithContext (Variables|Default|GetPreEventTimeBuffer%28Seconds)))
      ;[0 0.15 0.07284 0.5] Did the timer hit the PreEvent Time Buffer outside of Safe Zone Game Phase? Then Delay applying effects if so.
      (Default|UnbindEventfromOnSecondsRemainingUpdated _cachedtimermanager (delegate Custom|CheckForPreEventGate))
      (bind _returnvalue (Class|AthenaGameStateDataFunctionLibrary|GetAthenaGamePhase))
      (if (Utilities|Enum|Equal(Enum) _returnvalue "SafeZones")
        (CallFunction|ApplyPreEventEffects)
        (else
          (bind _returnvalue_2 (Utilities|Time|SetTimerbyEvent (EventDispatchers|CreateEvent :!event Custom|WaitForSafeZone) 1.0 true))
          (Variables|Default|SetWaitForSafeZoneTimer _returnvalue_2))))
    (else
      (Default|UnbindEventfromOnSecondsRemainingUpdated _cachedtimermanager (delegate Custom|CheckForPreEventGate)))))

(event Custom|WaitForSafeZone
  ;[0 0.15 0.07284 0.5] Did the timer hit the PreEvent Time Buffer outside of Safe Zone Game Phase? Then Delay applying effects if so.
  (bind _returnvalue (Class|AthenaGameStateDataFunctionLibrary|GetAthenaGamePhase))
  (if (Utilities|Enum|Equal(Enum) _returnvalue "SafeZones")
    (Utilities|Time|ClearandInvalidateTimerbyHandle (Variables|Default|GetWaitForSafeZoneTimer))
    (CallFunction|ApplyPreEventEffects)))

(event EventOnTimerFinished
  ; Remove Effects when Timer runs out and event starts.
  (if (Variables|Default|GetUsePreEventEffects)
    (CallFunction|RemovePreEventEffects)))

(fn ApplyPreEventEffects
  (GameplayEventRouter|BroadcastEvent (GameplayEventRouter|GetGlobalEventRouter) (Variables|Default|GetEventRouterEventName) 0))

(fn RemovePreEventEffects)

(fn ConstructionScript
  (|CallParentFunction))
