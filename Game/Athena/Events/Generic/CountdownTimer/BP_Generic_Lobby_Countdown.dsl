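; Generic event countdown hologram actor, used as a lobby timer or in match, that shows time remaining, blinks near the end, then fades out or shows "beginning soon" text.
; Timer manager delegates drive a Stage variable, with calendar event watching as fallback outside the lobby when no manager resolves.
; OnRep_Stage switches visibility between hidden, countdown, blinking and end-text states.
; A data table row supplies config and colors, pushed through material parameter collections.
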(event EventBeginPlay
  (bind _timermanagersoftref (Variables|Default|GetTimerManagerSoftRef))
  (bind _output (Utilities|ResolveSoftReference _timermanagersoftref))
  ;[0 0.15 0.004558 0.5] Set up Timer
  (|CallParentFunction)
  (CallFunction|SetupCountdownVisuals)
  ;; Check if we should use camera facing or not. Destroy actor if timer is disabled by config
  (Utilities|FlowControl|DelayUntilNextFrame)
  (if (Variables|Default|GetIsLobbyTimer)
    (if (not (Variables|Default|GetFrontendTimerEnabled))
      (Actor|DestroyActor)))
  (Rendering|Material|SetScalarParameterValue "/Game/Athena/Events/Generic/CountdownTimer/Materials/MPC_Countdown.MPC_Countdown" "CameraFacing" (Math|Float|SelectFloat 0.0 1.0 (Variables|Default|GetIsLobbyTimer)) :!class /Script/Engine.KismetMaterialLibrary)
  (Class|FortAsyncActionCurrentPlaylistReady|PlaylistDataReadyAsync
    (:then
      (Fortnite|FortShippingLog self "Begin play. BP_F_Countdown spawned")
      ;;[1 0.1159 0.9346] Create Localized Countdown Text & Bind to Events from Timer Manager to Trigger State Changes Based on Time Remaining
      (CallFunction|CreateLocalizedCountdownText)
      (Utilities|IsValid _output
        (:Is%20Valid
          (Default|BindEventtoOnSecondsRemainingUpdated _output (EventDispatchers|CreateEvent :!event Custom|Seconds%20Updated))
          (Default|BindEventtoOnTimerBlinking _output (EventDispatchers|CreateEvent :!event Custom|BlinkingStage))
          (Default|BindEventtoOnTimerFinished _output (EventDispatchers|CreateEvent :!event Custom|TimerFinishedPhase))
          (Default|BindEventtoOnTimerActive _output (EventDispatchers|CreateEvent :!event Custom|Stage1))
          (Class|FortAsyncActionAuthorityMatchReady|MatchReadyAsync
            (:OnReady
              _)))
        (:Is%20Not%20Valid
          (if (not (Variables|Default|GetIsLobbyTimer))
            (Variables|Default|SetStage 0)
            (Utilities|FlowControl|Delay 0.2) ; delay for logo widget to do its thing
            (CallFunction|StartWatchingCalendars))))
      (Variables|Default|SetStage 0))
    (:OnReady
      (Variables|Default|SetPlaylistReady true))))

(event Custom|Stage1
  ; Watch calendar events to drive timer stages
  (Variables|Default|SetStage 1))

(event Custom|StartWatchingCalendars
  ; Watch calendar events to drive timer stages
  (MCPContext|InitialCalendarSyncCompleted
    (:OnSyncComplete
      (CallFunction|CheckActiveCalendar)
      (MCPContext|WatchCalendarEvent (Variables|Default|GetCountdownEvent)
        (:then
          (MCPContext|WatchCalendarEvent (Variables|Default|GetBeginningSoonCalendarEvent)
            (:OnEventActive
              (Variables|Default|SetStage 4))))
        (:OnEventActive
          (Utilities|FlowControl|Sequence
            (:then_0
              (Fortnite|FortShippingLog self "BP_F_Countdown: Setting Time Via Clandar Event")
              (Variables|Default|SetStage 1))))
        (:OnEventEnded
          (if (not (Variables|Default|GetShowLogo))
            (Development|SetHiddeninGame (Variables|Default|GetLogoWidget) true true)))))))

(event Custom|Seconds%20Updated
  (params (SecondsRemaining :type Integer))
  ;[1 1 1] Update Timer Stages
  ;;[1 0.2113 0.9515] Two Minute Warning
  ;;;[0 0 0] if the seconds are getting ticked the event is on (or debug), go to stage 1 so we are visible
  (if (== (Variables|Default|GetStage) 0)
    (Variables|Default|SetStage 1))
  (if (and (<= SecondsRemaining 120) (not (Variables|Default|GetHas2MinWarningFired)))
    (Default|UnbindEventfromOnSecondsRemainingUpdated (Utilities|ResolveSoftReference (Variables|Default|GetTimerManagerSoftRef)) (delegate Custom|Seconds%20Updated))
    (Variables|Default|SetStage 2)
    (Variables|Default|SetHas2MinWarningFired true)))

(event Custom|CountdownStage
  ;[1 1 1] CountdownStage
  (Fortnite|FortShippingLog self "BP_F_Coundown: Stage 1: FX")
  (Rendering|SetActorHiddenInGame false)
  (Utilities|FlowControl|Delay 0.2)
  (Rendering|SetVisibility (Variables|Default|GetHologramCountdown) true false)
  (Rendering|SetVisibility (Variables|Default|GetWidget_LiveInText) false false)
  (if (not (Variables|Default|GetIsLobbyTimer))
    (Rendering|SetVisibility (Variables|Default|GetHologramIndicatorBeam) true false)))

(event Custom|UpdateVisuals
  (CallFunction|SetupCountdownVisuals))

(event Custom|BlinkingStage
  ;[1 1 1] Blinking Stage
  (Fortnite|FortShippingLog self "BP_F_Countdown: Stage 2: Turn off countdown timer")
  (bind _returnvalue (Utilities|Time|SetTimerbyEvent (EventDispatchers|CreateEvent :!event Custom|BlinkTimer) 1.0 true))
  (Variables|Default|SetBlinkHandle _returnvalue)
  (Rendering|SetVisibility (Variables|Default|GetWidget_LiveInText) false false)
  (Rendering|SetVisibility (Variables|Default|GetHologramCountdown) true false)
  (if (not (Variables|Default|GetIsLobbyTimer))
    (Rendering|SetVisibility (Variables|Default|GetHologramIndicatorBeam) true false)))

(event Custom|BlinkTimer
  ;[1 1 1] Blinking Timer Animation
  (Utilities|FlowControl|FlipFlop
    (:A
      (bind (_direction _newtrack_0) (Timeline|OpacityMult PlayFromStart)
        (:Update
          (Rendering|Material|SetScalarParameterValueonMaterials (Variables|Default|GetHologramCountdown) "OpacityMult" _newtrack_0)
          (bind _ascountdown_logo_widget (Utilities|Casting|CastToCountdownLogoWidget (UserInterface|GetWidget (Variables|Default|GetLogoWidget))))
          (Class|CountdownLogoWidget|UpdateOpacity _ascountdown_logo_widget _newtrack_0))))
    (:B
      (Timeline|OpacityMult Reverse))))

(event Custom|TimerFinishedPhase
  ;[1 1 1] Timer FInished Stage
  (if (Variables|Default|GetHideTimerOnFinish)
    (CallFunction|FadeOut false)
    (else
      (Variables|Default|SetStage 3)
      (CallFunction|ShowEndTimerText))))

(event Custom|FadeOut
  (params (Fade :type Boolean))
  ;[1 1 1] Fade out - Call in starting phase script
  (Fortnite|FortShippingLog self "BP_F_Countdown: FadeOut Event")
  (if Fade
    (Rendering|SetVisibility (Variables|Default|GetHologramCountdown) false false)
    (Rendering|SetVisibility (Variables|Default|GetHologramIndicatorBeam) false false)
    (bind (_direction _fade) (Timeline|Timeline_0 PlayFromStart)
      (:Update
        (Rendering|Material|SetScalarParameterValueonMaterials (Variables|Default|GetHologramCountdown) "OpacityMult" _fade)
        (bind _ascountdown_logo_widget (Utilities|Casting|CastToCountdownLogoWidget (UserInterface|GetWidget (Variables|Default|GetLogoWidget))))
        (Class|CountdownLogoWidget|UpdateOpacity _ascountdown_logo_widget _fade))
      (:Finished
        (switch Utilities|FlowControl|Switch|SwitchonETimelineDirection _direction
          (:Forward
            (Rendering|SetActorHiddenInGame true)
            (Utilities|Time|ClearandInvalidateTimerbyHandle (Variables|Default|GetBlinkHandle))
            (Actor|DestroyActor)))))
    (else
      (Rendering|SetActorHiddenInGame false)
      (Timeline|Timeline_0 Reverse))))

(event Custom|ScrubComplete
  ;[1 1 1] Fade out - Call in starting phase script
  (if (not (== (Variables|Default|GetStage) 3))
    (Timeline|Timeline_0 Stop)))

(event Custom|CreateLocalizedCountdownText
  ;[1 0.01658 0.0105] (not in use) localized texture update (Day/Hour/Minute text)
  (bind _userwidget (Class|FortAsyncActionCreateWidgetAsync|CreateWidgetAsync "/Game/Athena/Events/Generic/CountdownTimer/Widgets/LocalizedCountdownDisplay.LocalizedCountdownDisplay_C" 0 false)
    (:OnComplete
      (Utilities|IsValid _userwidget
        (:Is%20Valid
          (Rendering|DrawWidgettoRenderTarget _userwidget "/Game/Athena/Events/Generic/CountdownTimer/Textures/RT_LocalizedCountdownText.RT_LocalizedCountdownText"))))))

(fn CheckActiveCalendar
  (bind _allcalendarnames (Variables|Default|GetAllCalendarNames))
  (if (> (Utilities|Array|Length _allcalendarnames) 0)
    (bind _array_element (Utilities|Array|ForEachLoopwithBreak _allcalendarnames :!label _label)
      (:LoopBody
        (bind _biseventactive (CalendarEvents|GetCalendarEventInformation _array_element))
        (if _biseventactive
          (Variables|Default|SetActiveCalendarFound true)
          (Variables|Default|SetCountdownEvent _array_element)
          (goto _label Break)))
      (:Completed
        (if (not (Variables|Default|GetActiveCalendarFound))
          (Fortnite|FortShippingLog self "No Claendar Active"))))))

(fn OnRep_PlayAudio
  (if (== (Variables|Default|GetRepAudio) 0)
    (CallFunction|StopAudioInternal)))

(fn OnRep_PlayLoopingAudio)

(fn OnRep_Stage
  (bind _timermanagersoftref (Variables|Default|GetTimerManagerSoftRef))
  (bind _output (Utilities|ResolveSoftReference _timermanagersoftref))
  (switch int (Variables|Default|GetStage)
    (:0
      (Rendering|SetActorHiddenInGame true))
    (:1
      (CallFunction|CountdownStage))
    (:2
      (CallFunction|BlinkingStage))
    (:3
      (Utilities|IsValid _output
        (:Is%20Valid
          (Default|UnbindEventfromOnSecondsRemainingUpdated _output (EventDispatchers|CreateEvent :!event Seconds%20Updated)))))
    (:4
      (CallFunction|CountdownStage)
      (CallFunction|ShowEndTimerText))))

(fn PlayAudioInternal
  (params (StartTime :type Float%20%28double-precision%29))
  (Variables|Default|GetMusic
    (:then)
    (:else
      (bind _returnvalue (Audio|SpawnSoundatLocation :Sound (Variables|Default|GetCountdownMusic) :Location (Transformation|GetWorldLocation (Variables|Default|GetHologramCountdown)) :VolumeMultiplier 1.0 :PitchMultiplier 1.0 :StartTime StartTime :bAutoDestroy true))
      (Variables|Default|SetMusic _returnvalue))))

(fn SetupCountdownVisuals
  (bind _returnvalue (Utilities|GetDataTableRow "/Game/Athena/Events/Generic/CountdownTimer/DataTables/DT_CountdownVisuals.DT_CountdownVisuals" "CountdownVisuals")
    (:then
      (bind (_returnvalue_frontendtimerenabled_6_240c9d7a47bf0392e8bb38a8324b6a53 _returnvalue_showbeginningsoontext_31_80fd0b574243478b36578badf1f530d9 _returnvalue_showlogo_28_2cdeb0454720b6ee5de970b17e4374c5 _returnvalue_logourl_34_2421692f40990baaa7ff2194ac97b949 _returnvalue_logotransform_29_21f72a3e46219a6c414edb9251c9c790 _returnvalue_timer_color_numbers_13_daeb4a5c412953373f2b8cad5f10b9a2 _returnvalue_timer_color_text_15_6c06477740efa96b4a0ff1ac0d6643a5 _returnvalue_timer_color_image_17_86084b724ec62e0bdc718d871189fdce _returnvalue_timer_color_background_19_06b27dbf4495502dc0473f8450a93f10) (Utilities|Struct|BreakStructCountdownVisuals _returnvalue))
      (Variables|Default|SetFrontendTimerEnabled _returnvalue_frontendtimerenabled_6_240c9d7a47bf0392e8bb38a8324b6a53)
      (Variables|Default|SetShowLogo _returnvalue_showlogo_28_2cdeb0454720b6ee5de970b17e4374c5)
      (Variables|Default|SetHideTimerOnFinish (not _returnvalue_showbeginningsoontext_31_80fd0b574243478b36578badf1f530d9))
      (Transformation|SetRelativeTransform (Variables|Default|GetDHMSWidget) _returnvalue_logotransform_29_21f72a3e46219a6c414edb9251c9c790 false false)
      (Rendering|Material|SetVectorParameterValue "/Game/Athena/Events/Generic/CountdownTimer/Materials/MPC_Generic_Countdown_Visuals.MPC_Generic_Countdown_Visuals" "Timer_Color_Numbers" _returnvalue_timer_color_numbers_13_daeb4a5c412953373f2b8cad5f10b9a2 :!class /Script/Engine.KismetMaterialLibrary)
      (Rendering|Material|SetVectorParameterValue "/Game/Athena/Events/Generic/CountdownTimer/Materials/MPC_Generic_Countdown_Visuals.MPC_Generic_Countdown_Visuals" "Timer_Color_Text" _returnvalue_timer_color_text_15_6c06477740efa96b4a0ff1ac0d6643a5 :!class /Script/Engine.KismetMaterialLibrary)
      (Rendering|Material|SetVectorParameterValue "/Game/Athena/Events/Generic/CountdownTimer/Materials/MPC_Generic_Countdown_Visuals.MPC_Generic_Countdown_Visuals" "Timer_Color_Image" _returnvalue_timer_color_image_17_86084b724ec62e0bdc718d871189fdce :!class /Script/Engine.KismetMaterialLibrary)
      (Rendering|Material|SetVectorParameterValue "/Game/Athena/Events/Generic/CountdownTimer/Materials/MPC_Generic_Countdown_Visuals.MPC_Generic_Countdown_Visuals" "Timer_Color_Background" _returnvalue_timer_color_background_19_06b27dbf4495502dc0473f8450a93f10 :!class /Script/Engine.KismetMaterialLibrary))
    (:RowNotFound
      (Fortnite|FortShippingLog self "No row found for countdown visuals. Using defaults. "))))

(fn ShowEndTimerText
  (bind _asgalileo_countdown (Utilities|Casting|CastToGalileo_Countdown (UserInterface|GetUserWidgetObject (Variables|Default|GetWidget_LiveInText))))
  (bind _returnvalue (Rendering|Material|GetVectorParameterValue "/Game/Athena/Events/Generic/CountdownTimer/Materials/MPC_Generic_Countdown_Visuals.MPC_Generic_Countdown_Visuals" "Timer_Color_Text" :!class /Script/Engine.KismetMaterialLibrary))
  (Variables|/Script/UMG.TextBlock|SetText (Variables|Default|GetBeginningSoonText) (Variables|/Game/Athena/Prototype/Blueprints/Galileo/UI/Galileo_Countdown.Galileo_Countdown_C|GetText_LiveIn_Line1 _asgalileo_countdown))
  (Appearance|SetColorandOpacity (Variables|/Game/Athena/Prototype/Blueprints/Galileo/UI/Galileo_Countdown.Galileo_Countdown_C|GetText_LiveIn_Line1 _asgalileo_countdown) (Utilities|Struct|MakeSlateColor _returnvalue) :!class /Script/UMG.TextBlock)
  (Variables|/Script/UMG.TextBlock|SetText "" (Variables|/Game/Athena/Prototype/Blueprints/Galileo/UI/Galileo_Countdown.Galileo_Countdown_C|GetText_LiveIn_Line2 _asgalileo_countdown))
  (Rendering|SetVisibility (Variables|Default|GetWidget_LiveInText) true false)
  (Rendering|SetVisibility (Variables|Default|GetHologramCountdown) false false)
  (Development|SetHiddeninGame (Variables|Default|GetWidget_LiveInText) false false)
  (Rendering|SetVisibility (Variables|Default|GetHologramIndicatorBeam) false false))

(fn StopAudioInternal
  (bind _music (Variables|Default|GetMusic)
    (:then
      (Audio|Components|Audio|FadeOut _music 5.0 0.0 "Linear" :!class /Script/Engine.AudioComponent))
    (:else)))

(fn ConstructionScript
  (bind _hologramcountdown (Variables|Default|GetHologramCountdown))
  (|CallParentFunction)
  (CallFunction|SetupCountdownVisuals)
  (Rendering|Material|SetScalarParameterValue "/Game/Athena/Events/Generic/CountdownTimer/Materials/MPC_Countdown.MPC_Countdown" "CameraFacing" (Math|Float|SelectFloat 0.0 1.0 (Variables|Default|GetIsLobbyTimer)) :!class /Script/Engine.KismetMaterialLibrary)
  (Development|SetHiddeninGame _hologramcountdown false false)
  (Rendering|SetVisibility _hologramcountdown true false)
  (if (not (Variables|Default|GetIsLobbyTimer))
    (Rendering|SetVisibility (Variables|Default|GetHologramIndicatorBeam) true false)
    (if (not (Variables|Default|GetShowLogo))
      (Development|SetHiddeninGame (Variables|Default|GetLogoWidget) true false))))
