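; BP_Generic_TimerManager runs a replicated event countdown, fed by a calendar event, front-end lobby link preview, or fake/debug time.
; It stores the end in server world seconds; its RepNotify starts a one-second timer updating MPC_Countdown digit parameters.
; The authority picks stages from seconds-remaining thresholds and flags completion.
; Dispatchers announce changes, cancel or finish schedules visual clearing, and periodic calendar re-reads correct drift.
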(event EventBeginPlay
  ;[0 0 0] Begin Play - look for calendar event
  (Utilities|Casting|CastToFortGameStateFrontEnd (Game|GetGameState)
    (:then
      (Variables|Default|SetbIsFrontend true)
      (if (Variables|/Script/FortniteGame.FortRuntimeOptions|GetbEnableLobbyCountdownBasedOnLinkData (Options|GetRuntimeOptions))
        ; Front End / Link Service
        (bind _previewlink (Class|FortAsyncActionLinkEntryPreview|WatchLinkPreview)
          (:OnLinkPreviewUpdated
            (if (Utilities|IsValid _previewlink :!class /Script/Engine.KismetSystemLibrary)
              (if (Utilities|String|NotEqualExactly(String) (Class|FortGameActivity|GetLinkMnemonic _previewlink) (Variables|Default|GetLinkTimerActive))
                (Variables|Default|SetLinkTimerActive (Class|FortGameActivity|GetLinkMnemonic _previewlink))
                (if (DiscoverActivity|HasActiveLobbyCountdownTimer _previewlink :!class /Script/FortniteGame.FortGameActivity)
                  (Variables|Default|SetTimeUntilCountdownEnd (DiscoverActivity|GetTimeUntilNextLobbyCountdownTimerEnd _previewlink :!class /Script/FortniteGame.FortGameActivity))
                  (CallFunction|ContinueClockCountdown)
                  (Default|CallOnTimerActive)
                  (else
                    (Variables|Default|SetTimeUntilCountdownEnd (Math|Timespan|FromMinutes 0.0))
                    (Default|CallOnTimerInactive))))
              (else
                (Variables|Default|SetTimeUntilCountdownEnd (Math|Timespan|FromMinutes 0.0))
                (Variables|Default|SetLinkTimerActive "")
                (Default|CallOnTimerInactive)))))
        (else
          _)))
    (:CastFailed
      (bind _gamestate (Class|FortAsyncActionCurrentPlaylistReady|PlaylistDataReadyAsync)
        (:OnReady
          (if (not (Class|FortGameStateAthena|IsCreativeorPlaygroundModePlaylist _gamestate))
            (Class|FortAsyncActionAuthorityMatchReady|MatchReadyAsync
              (:OnReady
                _)))))))
  (if (not (Variables|TimerManager|GetFakeTimer))
    (bind (_eventname _timeuntilend) (MCPContext|WatchCalendarEvent (Variables|Default|GetCalendarEventName))
      (:OnEventActive
        (Variables|Default|SetTimeUntilCountdownEnd _timeuntilend)
        (CallFunction|ContinueClockCountdown))) ; Leadup to Festivus
  ))

(event Custom|ContinueClockCountdown
  ;[0 0 0] Set Server seconds
  (Variables|Default|SetServerWorldSecondsAtEvent (+ (GameState|GetServerWorldTimeSeconds (Game|GetGameState)) (Math|Timespan|GetTotalSeconds (Variables|Default|GetTimeUntilCountdownEnd)))))

(event Custom|DebugCountdown
  (params (Server%20Seconds :type Float%20%28double-precision%29))
  ;[1 0.1161 0.779] Cheat ke* DebugCountdown [seconds]
  (Variables|Default|SetTimeUntilCountdownEnd (Math|Timespan|FromSeconds Server%20Seconds))
  (Variables|Default|SetbTimerCancelled false)
  (CallFunction|ContinueClockCountdown))

(event Custom|CountdownTimer_New
  (bind _returnvalue (Game|GetGameState))
  ;[0 0 0] Update countdown
  (Utilities|IsValid _returnvalue
    (:Is%20Valid
      (bind _output_get (Variables|Default|SetSecondsRemaining (Utilities|ClampMinInteger (+ (Math|Float|Truncate (- (Variables|Default|GetServerWorldSecondsAtEvent) (GameState|GetServerWorldTimeSeconds _returnvalue))) (Variables|Default|GetTimeOffsetSeconds)))))
      (CallFunction|VisualUpdateEverySecond)
      (Default|CallOnSecondsRemainingUpdated _output_get)
      (CallFunction|UpdateStage)
      (if (<= _output_get (Variables|Default|GetTimeLeftToStartBlinking))
        (Default|CallOnTimerBlinking))
      ;;[1 1 1] count down is over show final graphic
      (if (<= _output_get 0)
        (Utilities|FlowControl|DoOnce false
          (:Completed
            (Networking|SwitchHasAuthority
              (:Authority
                (Variables|Default|SetbTimerFinished true)))))))))

(event Custom|TimerFinished
  ;[0 0 0] Update countdown
  ;;[1 1 1] count down is over show final graphic
  (Default|CallFadeOut)
  (Utilities|Time|ClearandInvalidateTimerbyHandle (Variables|Default|GetCountdownTimerHandle))
  (Rendering|Material|SetScalarParameterValue "/Game/Athena/Events/Generic/CountdownTimer/Materials/MPC_Countdown.MPC_Countdown" "ShowCountdown" 1.0 :!class /Script/Engine.KismetMaterialLibrary)
  (Rendering|Material|SetScalarParameterValue "/Game/Athena/Events/Generic/CountdownTimer/Materials/MPC_Countdown.MPC_Countdown" "ShowCountdownPhase2" 1.0 :!class /Script/Engine.KismetMaterialLibrary) ; Shows 2nd stage of countdown (hologram countdown floating in the sky)
  (Rendering|Material|SetScalarParameterValue "/Game/Athena/Events/Generic/CountdownTimer/Materials/MPC_Countdown.MPC_Countdown" "ShowCountdownPhase3" 1.0 :!class /Script/Engine.KismetMaterialLibrary) ; Shows post countdown graphics
  (Utilities|Time|SetTimerbyEvent (EventDispatchers|CreateEvent :!event Custom|ClearTimerVisuals) (Variables|Default|GetPhase3VisualDuration)))

(event Custom|Timer_TestCancel
  ;[0 0 0] Update countdown
  ;;[1 1 1] count down is over show final graphic
  (CallFunction|CancelTimer))

(event Custom|Clean%20Up%20Timer%20when%20Cancelling
  ;[0 0 0] Update countdown
  ;;[1 1 1] count down is over show final graphic
  (Utilities|Time|ClearandInvalidateTimerbyHandle (Variables|Default|GetCountdownTimerHandle))
  (Rendering|Material|SetScalarParameterValue "/Game/Athena/Events/Generic/CountdownTimer/Materials/MPC_Countdown.MPC_Countdown" "ShowCountdown" 1.0 :!class /Script/Engine.KismetMaterialLibrary)
  (Rendering|Material|SetScalarParameterValue "/Game/Athena/Events/Generic/CountdownTimer/Materials/MPC_Countdown.MPC_Countdown" "ShowCountdownPhase2" 1.0 :!class /Script/Engine.KismetMaterialLibrary) ; Shows 2nd stage of countdown (hologram countdown floating in the sky)
  (Rendering|Material|SetScalarParameterValue "/Game/Athena/Events/Generic/CountdownTimer/Materials/MPC_Countdown.MPC_Countdown" "ShowCountdownPhase3" 1.0 :!class /Script/Engine.KismetMaterialLibrary) ; Shows post countdown graphics
  (Utilities|Time|SetTimerbyEvent (EventDispatchers|CreateEvent :!event Custom|ClearTimerVisuals) (Variables|Default|GetPhase3VisualDuration)))

(event Custom|ErrorCorrection
  ;[0 0 0] Error correction for timer
  (Utilities|FlowControl|DoOnce false
    (:Completed
      (Utilities|Time|SetTimerbyEvent (EventDispatchers|CreateEvent :!event Custom|StartFinalAccuracyadjustment) (Math|Float|Clamp(Float) (Math|Conversions|ToFloat(Integer) (- (Utilities|ClampMinInteger (+ (Math|Float|Truncate (- (Variables|Default|GetServerWorldSecondsAtEvent) (GameState|GetServerWorldTimeSeconds (Game|GetGameState)))) (Variables|Default|GetTimeOffsetSeconds))) 300)) 0.001 1e+10)))))

(event Custom|ClearTimerVisuals
  ;[0 0 0] Clear Timer
  (Rendering|Material|SetScalarParameterValue "/Game/Athena/Events/Generic/CountdownTimer/Materials/MPC_Countdown.MPC_Countdown" "ShowCountdown" :!class /Script/Engine.KismetMaterialLibrary)
  (Rendering|Material|SetScalarParameterValue "/Game/Athena/Events/Generic/CountdownTimer/Materials/MPC_Countdown.MPC_Countdown" "ShowCountdownPhase2" :!class /Script/Engine.KismetMaterialLibrary) ; Shows 2nd stage of countdown (hologram countdown floating in the sky)
  (Rendering|Material|SetScalarParameterValue "/Game/Athena/Events/Generic/CountdownTimer/Materials/MPC_Countdown.MPC_Countdown" "ShowCountdownPhase3" :!class /Script/Engine.KismetMaterialLibrary) ; Shows post countdown graphics
)

(event Custom|StartFinalAccuracyadjustment
  ;[0 0 0] Start the accuracy adjustment
  (bind _returnvalue (Utilities|Time|SetTimerbyEvent (EventDispatchers|CreateEvent :!event Custom|CountDownTimer_AccuracyFix) 3e+01 true))
  (Variables|Default|SetCountDownTimerAccuracyfix _returnvalue))

(event Custom|CountDownTimer_AccuracyFix
  ;[0 0 0] Fix Accuracy every once in a wile based on timer ( more often in the last minutes)
  (if (Variables|TimerManager|GetFakeTimer)
    (Utilities|Time|ClearandInvalidateTimerbyHandle (Variables|Default|GetCountDownTimerAccuracyfix))
    (else
      (bind (_biseventactive _timesincebegin _timeuntilend) (CalendarEvents|GetCalendarEventInformation (Variables|Default|GetCalendarEventName)))
      (if _biseventactive
        (Variables|Default|SetTimeUntilCountdownEnd _timeuntilend)
        (Variables|Default|SetServerWorldSecondsAtEvent (+ (GameState|GetServerWorldTimeSeconds (Game|GetGameState)) (Math|Timespan|GetTotalSeconds (Variables|Default|GetTimeUntilCountdownEnd))))))))

(event Custom|VisualUpdateEverySecond
  (bind _days (Variables|Default|GetDays))
  (bind _seconds (Variables|Default|GetSeconds))
  (bind _minutes (Variables|Default|GetMinutes))
  (bind _hours (Variables|Default|GetHours))
  ;[0 0 0] Visual Update
  (if (not (Variables|Default|GetSkipVisualUpdate))
    (Utilities|FlowControl|Sequence
      (:then_0
        (bind _output_get (Variables|Default|SetTimespanRemaining (Math|Timespan|MakeTimespan 0 0 0 (Variables|Default|GetSecondsRemaining))))
        (bind (_output_get_days _output_get_hours _output_get_minutes _output_get_seconds) (Math|Timespan|BreakTimespan _output_get))
        (Variables|Default|SetDays _output_get_days)
        (Variables|Default|SetHours _output_get_hours)
        (Variables|Default|SetMinutes _output_get_minutes)
        (Variables|Default|SetSeconds _output_get_seconds))
      (:then_1
        (Rendering|Material|SetScalarParameterValue "/Game/Athena/Events/Generic/CountdownTimer/Materials/MPC_Countdown.MPC_Countdown" "ShowCountdownPhase2" 1.0 :!class /Script/Engine.KismetMaterialLibrary)
        (Rendering|Material|SetScalarParameterValue "/Game/Athena/Events/Generic/CountdownTimer/Materials/MPC_Countdown.MPC_Countdown" "ShowCountdownPhase3" :!class /Script/Engine.KismetMaterialLibrary)
        (Rendering|Material|SetScalarParameterValue "/Game/Athena/Events/Generic/CountdownTimer/Materials/MPC_Countdown.MPC_Countdown" "ShowCountdown" 1.0 :!class /Script/Engine.KismetMaterialLibrary)
        (Rendering|Material|SetScalarParameterValue "/Game/Athena/Events/Generic/CountdownTimer/Materials/MPC_Countdown.MPC_Countdown" "Digit1" (Math|Conversions|ToFloat(Integer) (Math|Integer|%25%28Integer%29 _seconds 10)) :!class /Script/Engine.KismetMaterialLibrary) ; seconds - ones
        (Rendering|Material|SetScalarParameterValue "/Game/Athena/Events/Generic/CountdownTimer/Materials/MPC_Countdown.MPC_Countdown" "Digit2" (Math|Conversions|ToFloat(Integer) (/ _seconds 10)) :!class /Script/Engine.KismetMaterialLibrary) ; seconds - tens
        (Rendering|Material|SetScalarParameterValue "/Game/Athena/Events/Generic/CountdownTimer/Materials/MPC_Countdown.MPC_Countdown" "Digit3" (Math|Conversions|ToFloat(Integer) (Math|Integer|%25%28Integer%29 _minutes 10)) :!class /Script/Engine.KismetMaterialLibrary) ; minutes - ones
        (Rendering|Material|SetScalarParameterValue "/Game/Athena/Events/Generic/CountdownTimer/Materials/MPC_Countdown.MPC_Countdown" "Digit4" (Math|Conversions|ToFloat(Integer) (/ _minutes 10)) :!class /Script/Engine.KismetMaterialLibrary) ; minutes - tens
        (Rendering|Material|SetScalarParameterValue "/Game/Athena/Events/Generic/CountdownTimer/Materials/MPC_Countdown.MPC_Countdown" "Digit5" (Math|Conversions|ToFloat(Integer) (Math|Integer|%25%28Integer%29 _hours 10)) :!class /Script/Engine.KismetMaterialLibrary) ; hours - ones
        (Rendering|Material|SetScalarParameterValue "/Game/Athena/Events/Generic/CountdownTimer/Materials/MPC_Countdown.MPC_Countdown" "Digit6" (Math|Conversions|ToFloat(Integer) (/ _hours 10)) :!class /Script/Engine.KismetMaterialLibrary) ; hours - tens
        (Rendering|Material|SetScalarParameterValue "/Game/Athena/Events/Generic/CountdownTimer/Materials/MPC_Countdown.MPC_Countdown" "Digit7" (Math|Conversions|ToFloat(Integer) (Math|Integer|%25%28Integer%29 _days 10)) :!class /Script/Engine.KismetMaterialLibrary) ; days - ones
        (Rendering|Material|SetScalarParameterValue "/Game/Athena/Events/Generic/CountdownTimer/Materials/MPC_Countdown.MPC_Countdown" "Digit8" (Math|Conversions|ToFloat(Integer) (/ _days 10)) :!class /Script/Engine.KismetMaterialLibrary) ; days - tens
      ))))

(fn CancelTimer
  (Variables|Default|SetbTimerCancelled true))

(fn CheckStage
  (params (SecondsRemaining :type Integer))
  (bind _newstate (LocalVariables|GetNewState))
  (LocalVariables|SetSecondsUntilEnd SecondsRemaining)
  ;[0 0 0] Client is authoritive in the front end
  (Networking|SwitchHasAuthority
    (:Authority
      ;[0 0 0] Find our current timer stage - look for the stage with the lowest seconds remaining that is high or equal to our current seconds remaining
      (for (_array_element _array_index) (Variables|Default|GetTimerStatesSecondsRemaining)
        (if (< (LocalVariables|GetSecondsUntilEnd) _array_element)
          (LocalVariables|SetNewState _array_index)))
      ;[0 0 0] If the stage has changed, update the stage to trigger the on timer stage changed delegate through the on rep
      (if (>= _newstate 0)
        (return _newstate)))))

(fn OnRep_bIsActive
  (Default|CallOnTimerActive))

(fn OnRep_bTimerCancelled
  (if (Variables|Default|GetbTimerCancelled)
    (CallFunction|CleanUpTimerWhenCancelling)))

(fn OnRep_bTimerFinished
  (Default|CallOnTimerFinished)
  (CallFunction|TimerFinished)
  (Development|PrintString "BP_Generic_TimerManager - Time Finished"))

(fn OnRep_CurrentState
  (Default|CallOnTimerStateChange (Variables|Default|GetCurrentState)))

(fn OnRep_ServerWorldSecondsAtEvent
  (Utilities|FlowControl|Sequence
    (:then_0
      ;[0 0 0] Call the timer active delegate when the timer is started
      (if (not (Variables|Default|GetbTimerActive))
        (Variables|Default|SetbTimerActive true)
        (Default|CallOnTimerActive)))
    (:then_1
      (if (<= (Variables|Default|GetServerWorldSecondsAtEvent) 0.0)
        (Rendering|Material|SetScalarParameterValue "/Game/Athena/Events/Generic/CountdownTimer/Materials/MPC_Countdown.MPC_Countdown" "ShowPhase2" :!class /Script/Engine.KismetMaterialLibrary)
        (elif (Variables|TimerManager|GetFakeTimer)
          (CallFunction|SetFakeTimerOffset)))
      (Networking|SwitchHasAuthority
        (:Authority
          (CallFunction|StartFinalAccuracyadjustment))
        (:Remote
          _))
      (if (not (Utilities|Time|IsTimerActivebyHandle (Variables|Default|GetCountdownTimerHandle)))
        (Variables|Default|SetSecondsRemaining (Utilities|ClampMinInteger (+ (Math|Float|Truncate (- (Variables|Default|GetServerWorldSecondsAtEvent) (GameState|GetServerWorldTimeSeconds (Game|GetGameState)))) (Variables|Default|GetTimeOffsetSeconds))))
        (bind _returnvalue (Utilities|Time|SetTimerbyEvent (EventDispatchers|CreateEvent :!event CountdownTimer_New) 1.0 true))
        (Variables|Default|SetCountdownTimerHandle _returnvalue)))))

(fn SetFakeTimerOffset
  (Variables|Default|SetTimeOffsetSeconds (Math|Float|Truncate (GameState|GetServerWorldTimeSeconds (Game|GetGameState)))))

(fn UpdateStage
  ;[0 0 0] Client is authoritive in the front end
  (Networking|SwitchHasAuthority
    (:Authority
      ;[0 0 0] Find our current timer stage - look for the stage with the lowest seconds remaining that is high or equal to our current seconds remaining
      (for (_array_element _array_index) (Variables|Default|GetTimerStatesSecondsRemaining)
        (if (and (or (< _array_element (LocalVariables|GetNewStateSecondsRemaining)) (< (LocalVariables|GetNewStateSecondsRemaining) 0)) (< (Variables|Default|GetSecondsRemaining) _array_element))
          (LocalVariables|SetNewStateSecondsRemaining _array_element)
          (LocalVariables|SetNewState _array_index)))
      ;[0 0 0] If the stage has changed, update the stage to trigger the on timer stage changed delegate through the on rep
      (if (and (>= (LocalVariables|GetNewState) 0) (!= (LocalVariables|GetNewState) (Variables|Default|GetCurrentState)))
        (Variables|Default|SetCurrentState (LocalVariables|GetNewState))))))

(fn UseFakeTimer
  (params (CountdownTime :type Float%20%28double-precision%29))
  (Variables|Default|SetbTimerCancelled false)
  (Variables|TimerManager|SetFakeTimer true)
  (Variables|Default|SetServerWorldSecondsAtEvent CountdownTime) ; Rep Notify\n..double click to follow the logic.
  (Rendering|SetActorHiddenInGame false))

(fn ConstructionScript)
